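; Cheat manager extension that adds debug commands for spawning premium supply drops.
; Both commands find the B_PremiumSupplyDrop_Manager actor in the world and print an error if it is missing.
; SpawnPremiumSupplyDrop has the manager spawn a drop at the impact point of the player's camera look trace.
; SpawnPremiumSupplyDropRandom instead calls the manager's debug EQS function for a random drop.
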(fn GetSupplyDropManager
  (bind _returnvalue (Actor|GetActorOfClass "/PremiumSupplyDrop/Gameplay/B_PremiumSupplyDrop_Manager.B_PremiumSupplyDrop_Manager_C"))
  (return _returnvalue))

(fn SpawnPremiumSupplyDrop
  (Development|PrintString "Calling Supply Premium Drop Cheat ")
  (bind _supplydropmanager (Default|GetSupplyDropManager))
  (Utilities|IsValid _supplydropmanager
    (:Is%20Valid
      (bind _playercontroller (Utilities|GetPlayer)
        (:Valid
          (bind _asfort_player_controller_athena (Utilities|Casting|CastToFortPlayerControllerAthena _playercontroller))
          (bind _outhitresult (Helpers|GetCameraLookHitResult _asfort_player_controller_athena 1e+05 false))
          (Spawning|SpawnSupplyDropActor _supplydropmanager (.ImpactPoint (Collision|BreakHitResult _outhitresult))))))
    (:Is%20Not%20Valid
      (Development|PrintString "Supply Premium Drop Cheat: Failed to find manager" true true "(R=1.000000,G=0.000000,B=0.070864,A=1.000000)"))))

(fn SpawnPremiumSupplyDropRandom
  (Development|PrintString "Calling Random Supply Premium Drop Cheat ")
  (bind _supplydropmanager (Default|GetSupplyDropManager))
  (Utilities|IsValid _supplydropmanager
    (:Is%20Valid
      (Class|BPremiumSupplyDropManager|DebugPremiumSupplyDropEQS _supplydropmanager))
    (:Is%20Not%20Valid
      (Development|PrintString "Supply Premium Drop Cheat: Failed to find manager" true true "(R=1.000000,G=0.000000,B=0.070864,A=1.000000)"))))
